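# Animation Blueprint for the M_MED_SulfurDean_Hear cosmetic body, layering head motion onto the body pose.
# It copies the attached parent's pose, blends in the Head mesh's pose from C_jaw, then runs the outfit's Control Rig.
# Head comes from the customization-loaded event, or on mannequins from a per-update search for the GuineaPig head mesh.
# Each update also flags whether either SulfurDean emote montage is playing and feeds that to the rig.
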AnimBlueprint /BRCosmetics/Characters/Player/Male/Medium/Bodies/M_MED_SulfurDean_Hear/Meshes/M_MED_SulfurDean_Hear_AnimBP
target_skeleton=&/M_MED_SulfurDean_Hear_Skeleton
graphs [
 AnimGraph {nodes [
   n1 = Get Head() -> (Head:Skeletal Mesh Component Budgeted Object Reference)
   n2 = Control Rig!(
    Source=Layered blend per bone(
     BasePose=Copy Pose From Mesh() {bUseAttachedParent=true}
     BlendPoses_0=Copy Pose From Mesh(SourceMeshComponent=@n1) {bUseAttachedParent=true}
    ) {layerSetup [{branchFilters [{boneName=C_jaw}]}]}
    Alpha
    SulfurEmotes:bool=Get SulfurDEanEmote() -> (SulfurDEanEmote:bool)
   ) -> (Pose) {
    alphaBoolBlend {blendOption=Linear}
    controlRigAssetReference {blueprintRigClass=&/M_MED_SulfurDean_Hear_CtrlRig.M_MED_SulfurDean_Hear_CtrlRig_C}
   }
   Output Pose(Result=@n2) {layerGroup=None}
  ]}
 EventGraph {nodes [
   Event BlueprintUpdateAnimation()
   > n1 = Sequence()
   n2 = GetOwningActor()
   > Cast To FortPlayerMannequin!(execute, Object=@n2) -> (then, CastFailed, AsFort Player Mannequin:Fort Player Mannequin Object Reference) {is=title K2Node_DynamicCast}
   > n3 = For Each Loop(Array:Array of Skeletal Mesh Component Object References=Get Components By Class!(self:Actor Object Reference=@n2, ComponentClass:Actor Component Class Reference=/Script/Engine.SkeletalMeshComponent) -> (ReturnValue:Array of Skeletal Mesh Component Object References)) -> (Array Element:Skeletal Mesh Component Object Reference)
   > Branch(Condition="Equal (Object)"(
     A=Get SkeletalMesh!(self:Skinned Mesh Component Object Reference=@n3) -> (SkeletalMesh:Skeletal Mesh Object Reference)
     B=/BRCosmetics/Characters/Player/Male/Medium/Heads/M_MED_GuineaPig_Head/Meshes/M_MED_GuineaPig_Head
    ))
   > n4 = Cast To SkeletalMeshComponentBudgeted!(execute, Object=@n3) -> (then, CastFailed, AsSkeletal Mesh Component Budgeted:Skeletal Mesh Component Budgeted Object Reference) {is=title K2Node_DynamicCast}
   > Set Head(Head:Skeletal Mesh Component Budgeted Object Reference=@n4) -> (Output_Get:Skeletal Mesh Component Budgeted Object Reference)
   n5 = GetAnimInstance(self=GetOwnerSkeletalMeshComponent())
   Is Valid(exec=@n1.then_1, InputObject=@n5)
   > Set SulfurDEanEmote(SulfurDEanEmote:bool=OR Boolean(
     A=Montage_IsPlaying(self=@n5, Montage=/BRCosmetics/Animation/Game/MainPlayer/Emotes/SulfurDean/CMM/Emote_SulfurDean_CMM_M)
     B=Montage_IsPlaying(self=@n5, Montage=/BRCosmetics/Animation/Game/MainPlayer/Emotes/SulfurDean_Hear/CMM/Emote_SulfurDean_Hear_Transform_CMM_M)
    )) -> (Output_Get:bool)
   n6 = Event OnCharacterCustomizationLoaded()
   > Set Head(Head=Get Array Item with EFort Custom Part Type(Array=@n6.CustomMeshParts, PartType=Head))
  ]}
]
variables [
 Head {type=SkeletalMeshComponentBudgeted}
 SulfurDEanEmote {type=boolean}
]
